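; Gameplay ability for the Athena Bush consumable that, going by its names, manages aim assist against the bush.
; On activation it runs generic ability setup, whose output involves an AutoAimConditionFailed task and applying the GE_Bush_AimAssistEnabled effect to the owner.
; The task's failure callback has no logic attached.
; On end it removes owner effects by granted tag, but the tag supplied is empty.
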(event Ability|EventActivateAbility
  (|CallParentFunction)
  (Setup|SetupGenericAbility false false
    (:Output
      (Ability|Tasks|AutoAimConditionFailed
        (:then
          _)
        (:OnAutoAimConditionFailed
          _))
      (Ability|ApplyGameplayEffectToOwner "/Game/Athena/Items/Consumables/Bush/GE_Bush_AimAssistEnabled.GE_Bush_AimAssistEnabled_C" 0))))

(event Ability|EventOnEndAbility
  (params (bWasCancelled :type Boolean))
  (|CallParentFunction bWasCancelled)
  (Ability|RemoveGameplayEffectFromOwnerWithGrantedTags (GameplayTags|MakeGameplayTagContainerfromTag "(TagName=\"\")")))
